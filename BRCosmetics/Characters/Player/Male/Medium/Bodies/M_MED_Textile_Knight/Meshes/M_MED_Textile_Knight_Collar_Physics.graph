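# Physics asset for the collar of the medium male Textile Knight player body, setting up collision and joints for collar motion.
# Eight collar segments around the neck are boxes, each jointed to a tiny _MSTR anchor capsule with limited swing and very narrow twist.
# The anchors join spine_05 through angularly free constraints, as does a neck_02-to-head chain.
# Capsules on head, neck and upper spine give the body's shape beside the collar.
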PhysicsAsset /BRCosmetics/Characters/Player/Male/Medium/Bodies/M_MED_Textile_Knight/Meshes/M_MED_Textile_Knight_Collar_Physics
bodies {
 dyn_collar_BK [{shapeType=Box, extentX=5, extentY=2.0998, extentZ=9.0834, shapeName=dyn_collar_BK_box, center=<4.5 0 0>, rotation=<rot 0 0 90>}]
 dyn_collar_BK_MSTR [{shapeType=Capsule, length=0.505, radius=0.505, shapeName=dyn_collar_BK_MSTR_capsule, rotation=<rot 0 -90 0>}]
 dyn_collar_BK_l [{shapeType=Box, extentX=5, extentY=2, extentZ=7.0346, shapeName=dyn_collar_BK_l_box, center=<4.5 0 0>, rotation=<rot 0 0 88.0685>}]
 dyn_collar_BK_l_MSTR [{shapeType=Capsule, length=0.505, radius=0.505, shapeName=dyn_collar_BK_l_MSTR_capsule, rotation=<rot 0 -90 0>}]
 dyn_collar_BK_r [{shapeType=Box, extentX=5, extentY=2, extentZ=8.1393, shapeName=dyn_collar_BK_r_box, center=<4.5 0 0>, rotation=<rot 0 0 101.2785>}]
 dyn_collar_BK_r_MSTR [{shapeType=Capsule, length=0.505, radius=0.505, shapeName=dyn_collar_BK_r_MSTR_capsule, rotation=<rot 0 -90 0>}]
 dyn_collar_fr [{shapeType=Box, extentX=5, extentY=2, extentZ=11.4591, shapeName=dyn_collar_fr_box, center=<4.5 0 0>, rotation=<rot 0 0 90>}]
 dyn_collar_fr_MSTR [{shapeType=Capsule, length=0.505, radius=0.505, shapeName=dyn_collar_fr_MSTR_capsule, rotation=<rot 0 -90 0>}]
 dyn_collar_fr_l [{shapeType=Box, extentX=5, extentY=2, extentZ=6.805, shapeName=dyn_collar_fr_l_box, center=<4.5 0 0>, rotation=<rot 0 0 90>}]
 dyn_collar_fr_l_MSTR [{shapeType=Capsule, length=0.505, radius=0.505, shapeName=dyn_collar_fr_l_MSTR_capsule, rotation=<rot 0 -90 0>}]
 dyn_collar_fr_r [{shapeType=Box, extentX=5, extentY=2, extentZ=6.895, shapeName=dyn_collar_fr_r_box, center=<4.5 0 0>, rotation=<rot 0 0 90>}]
 dyn_collar_fr_r_MSTR [{shapeType=Capsule, length=0.505, radius=0.505, shapeName=dyn_collar_fr_r_MSTR_capsule, rotation=<rot 0 -90 0>}]
 dyn_collar_l [{shapeType=Box, extentX=5, extentY=2, extentZ=6.5894, shapeName=dyn_collar_l_box, center=<4.5 0 0>, rotation=<rot 0 0 90>}]
 dyn_collar_l_MSTR [{shapeType=Capsule, length=0.505, radius=0.505, shapeName=dyn_collar_l_MSTR_capsule, rotation=<rot 0 -90 0>}]
 dyn_collar_r [{shapeType=Box, extentX=5, extentY=2, extentZ=7.0139, shapeName=dyn_collar_r_box, center=<4.5 0 0>, rotation=<rot 0 0 90>}]
 dyn_collar_r_MSTR [{shapeType=Capsule, length=0.505, radius=0.505, shapeName=dyn_collar_r_MSTR_capsule, rotation=<rot 0 -90 0>}]
 head [{shapeType=Capsule, length=5.003, radius=10.4271, shapeName=head_capsule, center=<2.6728 4.5145 0>, rotation=<rot 90 20.556 71.5897>}]
 neck_02 [{shapeType=Capsule, length=6.3566, radius=6.0332, shapeName=neck_02_capsule, center=<3.4946 -0.1394 -0.0001>, rotation=<rot 90 0 7.7168>}]
 spine_05 [{shapeType=Capsule, length=2, radius=2, shapeName=spine_05_capsule}]
}
constraints [
 {bone1Name=dyn_collar_fr_l, bone2Name=dyn_collar_fr_l_MSTR, swing1LimitDegrees=15, swing2LimitDegrees=35, twistLimitDegrees=3}
 {bone1Name=dyn_collar_fr_r, bone2Name=dyn_collar_fr_r_MSTR, swing1LimitDegrees=15, swing2LimitDegrees=35, twistLimitDegrees=3}
 {bone1Name=dyn_collar_r, bone2Name=dyn_collar_r_MSTR, swing1LimitDegrees=15, swing2LimitDegrees=35, twistLimitDegrees=3}
 {bone1Name=dyn_collar_BK_l, bone2Name=dyn_collar_BK_l_MSTR, swing1LimitDegrees=15, swing2LimitDegrees=35, twistLimitDegrees=3}
 {bone1Name=dyn_collar_BK_r, bone2Name=dyn_collar_BK_r_MSTR, swing1LimitDegrees=15, swing2LimitDegrees=35, twistLimitDegrees=3}
 {bone1Name=dyn_collar_fr, bone2Name=dyn_collar_fr_MSTR, swing1LimitDegrees=15, swing2LimitDegrees=35, twistLimitDegrees=3}
 {bone1Name=dyn_collar_l, bone2Name=dyn_collar_l_MSTR, swing1LimitDegrees=15, swing2LimitDegrees=35, twistLimitDegrees=3}
 {bone1Name=dyn_collar_BK, bone2Name=dyn_collar_BK_MSTR, swing1LimitDegrees=15, swing2LimitDegrees=35, twistLimitDegrees=3}
 {
  bone1Name=head
  bone2Name=neck_02
  swing1LimitDegrees=45
  swing1Motion=Free
  swing2LimitDegrees=45
  swing2Motion=Free
  twistLimitDegrees=45
  twistMotion=Free
 }
 {
  bone1Name=neck_02
  bone2Name=spine_05
  swing1LimitDegrees=45
  swing1Motion=Free
  swing2LimitDegrees=45
  swing2Motion=Free
  twistLimitDegrees=45
  twistMotion=Free
 }
 {
  bone1Name=dyn_collar_BK_MSTR
  bone2Name=spine_05
  swing1LimitDegrees=45
  swing1Motion=Free
  swing2LimitDegrees=45
  swing2Motion=Free
  twistLimitDegrees=45
  twistMotion=Free
 }
 {
  bone1Name=dyn_collar_l_MSTR
  bone2Name=spine_05
  swing1LimitDegrees=45
  swing1Motion=Free
  swing2LimitDegrees=45
  swing2Motion=Free
  twistLimitDegrees=45
  twistMotion=Free
 }
 {
  bone1Name=dyn_collar_fr_MSTR
  bone2Name=spine_05
  swing1LimitDegrees=45
  swing1Motion=Free
  swing2LimitDegrees=45
  swing2Motion=Free
  twistLimitDegrees=45
  twistMotion=Free
 }
 {
  bone1Name=dyn_collar_BK_r_MSTR
  bone2Name=spine_05
  swing1LimitDegrees=45
  swing1Motion=Free
  swing2LimitDegrees=45
  swing2Motion=Free
  twistLimitDegrees=45
  twistMotion=Free
 }
 {
  bone1Name=dyn_collar_BK_l_MSTR
  bone2Name=spine_05
  swing1LimitDegrees=45
  swing1Motion=Free
  swing2LimitDegrees=45
  swing2Motion=Free
  twistLimitDegrees=45
  twistMotion=Free
 }
 {
  bone1Name=dyn_collar_r_MSTR
  bone2Name=spine_05
  swing1LimitDegrees=45
  swing1Motion=Free
  swing2LimitDegrees=45
  swing2Motion=Free
  twistLimitDegrees=45
  twistMotion=Free
 }
 {
  bone1Name=dyn_collar_fr_r_MSTR
  bone2Name=spine_05
  swing1LimitDegrees=45
  swing1Motion=Free
  swing2LimitDegrees=45
  swing2Motion=Free
  twistLimitDegrees=45
  twistMotion=Free
 }
 {
  bone1Name=dyn_collar_fr_l_MSTR
  bone2Name=spine_05
  swing1LimitDegrees=45
  swing1Motion=Free
  swing2LimitDegrees=45
  swing2Motion=Free
  twistLimitDegrees=45
  twistMotion=Free
 }
]
