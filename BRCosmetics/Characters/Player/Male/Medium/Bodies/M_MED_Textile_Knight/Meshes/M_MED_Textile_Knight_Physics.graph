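# Physics asset for the Textile Knight body, a male medium player cosmetic, covering the back shirt flaps and the lower body around them.
# Three box bodies on the center, left and right dyn_shirt_bk bones hang from the pelvis with narrow twist and moderate swing limits, the side boxes angled outward.
# Capsules on the pelvis and both thighs give the body volume, with each thigh joined to the pelvis by free swing and twist.
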PhysicsAsset /BRCosmetics/Characters/Player/Male/Medium/Bodies/M_MED_Textile_Knight/Meshes/M_MED_Textile_Knight_Physics
bodies {
 dyn_shirt_bk [{shapeType=Box, extentX=15, extentY=15, extentZ=3.6663, shapeName=dyn_shirt_bk_box, center=<5 0 0>}]
 dyn_shirt_bk_l [{shapeType=Box, extentX=15, extentY=15, extentZ=2.5, shapeName=dyn_shirt_bk_l_box, center=<5 0 0>, rotation=<rot 0 0 40>}]
 dyn_shirt_bk_r [{shapeType=Box, extentX=15, extentY=15, extentZ=2.5, shapeName=dyn_shirt_bk_r_box, center=<5 0 0>, rotation=<rot 0 0 -40>}]
 pelvis [{shapeType=Capsule, length=4.9538, radius=12.9135, center=<-6.6404 0 0.0003>}]
 thigh_l [{shapeType=Capsule, length=23.2784, radius=10.2216, shapeName=thigh_l_capsule, center=<-15 0 0>, rotation=<rot 90 0 0>}]
 thigh_r [{shapeType=Capsule, length=24, radius=9.5, shapeName=thigh_r_capsule, center=<15 -0.0004 -0.0002>, rotation=<rot 90 0 0>}]
}
constraints [
 {bone1Name=dyn_shirt_bk_r, bone2Name=pelvis, swing1LimitDegrees=20, swing2LimitDegrees=45, twistLimitDegrees=3}
 {bone1Name=dyn_shirt_bk_l, bone2Name=pelvis, swing1LimitDegrees=20, swing2LimitDegrees=45, twistLimitDegrees=3}
 {bone1Name=dyn_shirt_bk, bone2Name=pelvis, swing1LimitDegrees=20, swing2LimitDegrees=45, twistLimitDegrees=3}
 {
  bone1Name=thigh_l
  bone2Name=pelvis
  swing1LimitDegrees=45
  swing1Motion=Free
  swing2LimitDegrees=45
  swing2Motion=Free
  twistLimitDegrees=45
  twistMotion=Free
 }
 {
  bone1Name=thigh_r
  bone2Name=pelvis
  swing1LimitDegrees=45
  swing1Motion=Free
  swing2LimitDegrees=45
  swing2Motion=Free
  twistLimitDegrees=45
  twistMotion=Free
 }
]
