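# Control Rig for the male medium Textile Knight body that procedurally moves collar bones as the head turns, likely to avoid clipping.
# It reads the head's rotation relative to spine_05 and converts it to ZYX Euler angles.
# Clamped remaps of negative Z angle drive one ModifyTransforms node, offsetting dyn_collar_BK_r in translation and rotation and another item in translation only.
# The correction responds to one direction of rotation only.
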ControlRigBlueprint /BRCosmetics/Characters/Player/Male/Medium/Bodies/M_MED_Textile_Knight/Meshes/M_MED_Textile_Knight_target_CtrlRig
graphs [
 RigVMFunctionLibrary {links [], nodes []}
 RigVMModel {nodes [
   RigUnit_BeginExecution = &RigUnit_BeginExecution()
   n1 = &RigVMFunction_MathQuaternionToEuler(
    RotationOrder=ZYX
    Value=&RigUnit_GetRelativeTransformForItem(Child=(Type=Bone,Name="head"), Parent=(Type=Bone,Name="spine_05")).RelativeTransform.Rotation
   )
   n2 = &RigVMFunction_MathFloatRemap(SourceMaximum=-60, TargetMaximum=9, bClamp=true, Value=@n1.Result.Z)
   &RigUnit_ModifyTransforms(
    ItemToModify="((Item=(Type=Bone,Name=\"dyn_collar_BK_r\"),Transform=(Rotation=(X=0,Y=0,Z=0,W=1),Translation=(X=0,Y=9,Z=0),Scale3D=(X=1,Y=1,Z=1))),(Item=(Type=Bone,Name=\"None\"),Transform=(Rotation=(X=0,Y=0,Z=0,W=1),Translation=(X=0,Y=0,Z=0),Scale3D=(X=1,Y=1,Z=1))),(Item=(T ...(truncated)"
    Weight=1
    ItemToModify.0.Transform.Translation.Y=@n2.Result
    ItemToModify.2.Transform.Translation.Y=@n2.Result
    ItemToModify.2.Transform.Translation.Z=&RigVMFunction_MathFloatRemap(SourceMaximum=-60, TargetMaximum=3, bClamp=true, Value=@n1.Result.Z).Result
    ItemToModify.0.Transform.Rotation=&RigVMFunction_MathQuaternionFromEuler(
     Euler=(X=0,Y=0,Z=0)
     Euler.Y=&RigVMFunction_MathFloatRemap(SourceMaximum=-60, TargetMaximum=-15, bClamp=true, Value=@n1.Result.Z).Result
    ).Result
    ExecutePin=@RigUnit_BeginExecution.ExecutePin
   )
  ]}
]
hierarchy [
 root {loc=<0 0 0>, rot=<0 0 0 1>}
 pelvis {parent=root, loc=<0 -0.7071 93.1191>, rot=<0.0127 0.707 -0.0127 -0.707>}
 spine_01 {parent=pelvis, loc=<6.7494 -0.6388 0>, rot=<0 0 0.034 -0.9994>}
 spine_02 {parent=spine_01, loc=<7.2518 -0.035 0>, rot=<0 0 0.0238 -0.9997>}
 spine_03 {parent=spine_02, loc=<6.0104 0 0>, rot=<0 0 0.0149 -0.9999>}
 spine_04 {parent=spine_03, loc=<8.1024 0 0>, rot=<0 0 -0.0132 -0.9999>}
 spine_05 {parent=spine_04, loc=<10.0146 0 0>, rot=<0 0 -0.0015 -1>}
 clavicle_l {parent=spine_05, loc=<4.3575 -0.1466 -2.3223>, rot=<0.0465 -0.7289 -0.009 -0.6829>}
 upperarm_l {parent=clavicle_l, loc=<15.5111 0 0>, rot=<0.0226 -0.3234 -0.0088 -0.9459>}
 lowerarm_l {parent=upperarm_l, loc=<26.6514 0 0>, rot=<0.0391 -0.0099 -0.2453 -0.9686>}
 hand_l {parent=lowerarm_l, loc=<22.3491 0 0>, rot=<-0.7081 -0.0507 0.0975 -0.6975>}
 index_metacarpal_l {parent=hand_l, loc=<1.6535 0.0183 -2.96>, rot=<0.0158 -0.0124 -0.0173 -0.9996>}
 index_01_l {parent=index_metacarpal_l, loc=<8.3008 0.0357 -0.0524>, rot=<-0.0044 0.0485 0.0924 -0.9945>}
 index_02_l {parent=index_01_l, loc=<3.3587 -0.3004 0.0452>, rot=<-0.002 0.0027 0.3225 -0.9466>}
 index_03_l {parent=index_02_l, loc=<2.5578 -0.123 -0.0252>, rot=<-0.002 0.0067 0.2558 -0.9667>}
 middle_metacarpal_l {parent=hand_l, loc=<2.1856 -0.1725 -0.5984>, rot=<0.006 0.0168 -0.0329 -0.9993>}
 middle_01_l {parent=middle_metacarpal_l, loc=<7.3275 -0.2447 0.0129>, rot=<-0.0208 0.0042 0.0858 -0.9961>}
 middle_02_l {parent=middle_01_l, loc=<4.0123 0.1401 0.1175>, rot=<-0.0218 0.0258 0.3512 -0.9357>}
 middle_03_l {parent=middle_02_l, loc=<2.6077 0.0534 0.1373>, rot=<0.0082 0.0188 0.1708 -0.9851>}
 pinky_metacarpal_l {parent=hand_l, loc=<1.7616 -0.9753 3.3814>, rot=<-0.0397 0.0814 -0.069 -0.9935>}
 pinky_01_l {parent=pinky_metacarpal_l, loc=<7.1709 -0.4831 -0.1094>, rot=<-0.0654 -0.0035 0.2242 -0.9723>}
 pinky_02_l {parent=pinky_01_l, loc=<2.8274 0.0893 -0.1043>, rot=<0.0037 0.0011 0.2854 -0.9584>}
 pinky_03_l {parent=pinky_02_l, loc=<2.2046 0.0307 -0.0078>, rot=<-0.0002 0.0098 0.1445 -0.9895>}
 ring_metacarpal_l {parent=hand_l, loc=<2.2074 -0.3625 1.3638>, rot=<-0.0284 0.0722 -0.0458 -0.9959>}
 ring_01_l {parent=ring_metacarpal_l, loc=<7.0295 -0.6197 -0.1446>, rot=<-0.0489 -0.0252 0.1892 -0.9804>}
 ring_02_l {parent=ring_01_l, loc=<3.6174 0.4647 -0.1027>, rot=<0 0 0.2806 -0.9598>}
 ring_03_l {parent=ring_02_l, loc=<2.4152 0.114 0.0199>, rot=<0 0 0.1189 -0.9929>}
 thumb_01_l {parent=hand_l, loc=<1.3118 -0.3191 -2.459>, rot=<0.6546 -0.4042 0.0664 -0.6354>}
 thumb_02_l {parent=thumb_01_l, loc=<5.9336 -0.0448 0.0401>, rot=<-0.0037 0.0213 0.298 -0.9543>}
 thumb_03_l {parent=thumb_02_l, loc=<2.2223 0.3682 -0.0907>, rot=<0.0118 -0.013 0.0066 -0.9998>}
 lowerarm_twist_01_l {parent=lowerarm_l, loc=<17.8549 0 0>, rot=<0 0 0 -1>}
 lowerarm_twist_02_l {parent=lowerarm_l, loc=<9.6199 0 0>, rot=<0 0 0 -1>}
 upperarm_twist_01_l {parent=upperarm_l, loc=<0.0041 0 0>, rot=<0 0 0 -1>}
 deform_deltoid_fr_l {parent=upperarm_twist_01_l, loc=<2.4511 0 0>, rot=<0 0 0 -1>}
 upperarm_twist_02_l {parent=upperarm_l, loc=<13.7941 0 0>, rot=<0 0 0 -1>}
 deform_pectoral_fr_l {parent=clavicle_l, loc=<9.3905 0 0>, rot=<0 0 0 -1>}
 clavicle_r {parent=spine_05, loc=<4.3575 -0.1466 2.3223>, rot=<0.7289 0.0465 -0.6829 0.009>}
 upperarm_r {parent=clavicle_r, loc=<-15.5111 0 0>, rot=<0.0226 -0.3234 -0.0088 -0.9459>}
 lowerarm_r {parent=upperarm_r, loc=<-26.6515 0 0>, rot=<0.0391 -0.0099 -0.2453 -0.9686>}
 hand_r {parent=lowerarm_r, loc=<-22.3492 0 0>, rot=<-0.7081 -0.0507 0.0975 -0.6975>}
 index_metacarpal_r {parent=hand_r, loc=<-1.6534 -0.0181 2.9601>, rot=<0.0158 -0.0124 -0.0173 -0.9996>}
 index_01_r {parent=index_metacarpal_r, loc=<-8.3008 -0.0357 0.0524>, rot=<-0.0044 0.0485 0.0924 -0.9945>}
 index_02_r {parent=index_01_r, loc=<-3.3589 0.3003 -0.0453>, rot=<-0.002 0.0027 0.3225 -0.9466>}
 index_03_r {parent=index_02_r, loc=<-2.5578 0.1231 0.0251>, rot=<-0.002 0.0067 0.2558 -0.9667>}
 middle_metacarpal_r {parent=hand_r, loc=<-2.1856 0.1726 0.5984>, rot=<0.006 0.0168 -0.0329 -0.9993>}
 middle_01_r {parent=middle_metacarpal_r, loc=<-7.3273 0.2447 -0.0129>, rot=<-0.0208 0.0042 0.0858 -0.9961>}
 middle_02_r {parent=middle_01_r, loc=<-4.0142 -0.1411 -0.07>, rot=<-0.0217 0.0259 0.3512 -0.9357>}
 middle_03_r {parent=middle_02_r, loc=<-2.6074 -0.0534 -0.1312>, rot=<0.0083 0.0188 0.1708 -0.9851>}
 pinky_metacarpal_r {parent=hand_r, loc=<-1.7617 0.9753 -3.3814>, rot=<-0.0397 0.0814 -0.069 -0.9935>}
 pinky_01_r {parent=pinky_metacarpal_r, loc=<-7.1707 0.4832 0.1094>, rot=<-0.0654 -0.0035 0.2242 -0.9723>}
 pinky_02_r {parent=pinky_01_r, loc=<-2.8274 -0.0892 0.1043>, rot=<0.0037 0.0011 0.2854 -0.9584>}
 pinky_03_r {parent=pinky_02_r, loc=<-2.2045 -0.031 0.0078>, rot=<-0.0002 0.0098 0.1445 -0.9895>}
 ring_metacarpal_r {parent=hand_r, loc=<-2.2073 0.3627 -1.3638>, rot=<-0.0284 0.0722 -0.0458 -0.9959>}
 ring_01_r {parent=ring_metacarpal_r, loc=<-7.0297 0.6197 0.1446>, rot=<-0.0489 -0.0252 0.1892 -0.9804>}
 ring_02_r {parent=ring_01_r, loc=<-3.6171 -0.4648 0.1027>, rot=<0 0 0.2806 -0.9598>}
 ring_03_r {parent=ring_02_r, loc=<-2.4153 -0.1141 -0.0199>, rot=<0 0 0.1189 -0.9929>}
 thumb_01_r {parent=hand_r, loc=<-1.3119 0.3192 2.459>, rot=<0.6546 -0.4042 0.0664 -0.6354>}
 thumb_02_r {parent=thumb_01_r, loc=<-5.9336 0.0447 -0.0401>, rot=<-0.0037 0.0213 0.298 -0.9543>}
 thumb_03_r {parent=thumb_02_r, loc=<-2.2223 -0.3682 0.0907>, rot=<0.0118 -0.013 0.0066 -0.9998>}
 lowerarm_twist_01_r {parent=lowerarm_r, loc=<-17.855 0 0>, rot=<0 0 0 1>}
 lowerarm_twist_02_r {parent=lowerarm_r, loc=<-9.6199 0 0>, rot=<0 0 0 1>}
 upperarm_twist_01_r {parent=upperarm_r, loc=<-0.0041 0 0>, rot=<0 0 0 1>}
 deform_deltoid_fr_r {parent=upperarm_twist_01_r, loc=<-2.4512 0 0>, rot=<0 0 0 1>}
 upperarm_twist_02_r {parent=upperarm_r, loc=<-13.7942 0 0>, rot=<0 0 0 1>}
 deform_pectoral_fr_r {parent=clavicle_r, loc=<-9.39 0 0>, rot=<0 0 0 1>}
 neck_01 {parent=spine_05, loc=<11.3138 0.1529 0>, rot=<0 0 -0.2003 -0.9797>}
 neck_02 {parent=neck_01, loc=<4.5627 0 0>, rot=<0 0 0.0504 -0.9987>}
 head {parent=neck_02, loc=<5.3914 0 0>, rot=<0 0 0.115 -0.9934>}
 dyn_collar_BK_MSTR {parent=spine_05, loc=<15.2791 -4.7795 0.1587>, rot=<-0.7057 -0.0439 -0.0439 -0.7057>}
 dyn_collar_BK {parent=dyn_collar_BK_MSTR, loc=<0 0 0>, rot=<0 0 0 -1>}
 dyn_collar_l_MSTR {parent=spine_05, loc=<11.1012 3.8973 -8.0509>, rot=<0.9608 0.2495 -0.1206 -0.0129>}
 dyn_collar_l {parent=dyn_collar_l_MSTR, loc=<0 0 0>, rot=<0 0 0 1>}
 dyn_collar_fr_MSTR {parent=spine_05, loc=<6.662 9.4948 -0.4934>, rot=<0.656 0.2839 -0.291 -0.636>}
 dyn_collar_fr {parent=dyn_collar_fr_MSTR, loc=<0 0 0>, rot=<0 0 0 1>}
 dyn_collar_BK_r_MSTR {parent=spine_05, loc=<13.5291 -1.136 7.4002>, rot=<-0.3523 0.0445 -0.1908 -0.9152>}
 dyn_collar_BK_r {parent=dyn_collar_BK_r_MSTR, loc=<0 0 0>, rot=<0 0 0 -1>}
 dyn_collar_BK_l_MSTR {parent=spine_05, loc=<13.5291 -1.136 -7.4>, rot=<0.9438 0.1804 -0.0295 0.2755>}
 dyn_collar_BK_l {parent=dyn_collar_BK_l_MSTR, loc=<0 0 0>, rot=<0 0 0 1>}
 dyn_collar_r_MSTR {parent=spine_05, loc=<11.1012 3.8973 8.051>, rot=<0.001 0.1175 -0.2228 -0.9678>}
 dyn_collar_r {parent=dyn_collar_r_MSTR, loc=<0 0 0>, rot=<0 0 0 -1>}
 dyn_collar_fr_r_MSTR {parent=spine_05, loc=<8.1245 7.9728 5.528>, rot=<0.361 0.1704 -0.3182 -0.8599>}
 dyn_collar_fr_r {parent=dyn_collar_fr_r_MSTR, loc=<0 0 0>, rot=<0 0 0 -1>}
 dyn_collar_fr_l_MSTR {parent=spine_05, loc=<8.1245 7.9728 -5.5279>, rot=<0.8767 0.3379 -0.1675 -0.2987>}
 dyn_collar_fr_l {parent=dyn_collar_fr_l_MSTR, loc=<0 0 0>, rot=<0 0 0 -1>}
 thigh_l {parent=pelvis, loc=<-2.5693 -0.0109 -7.7501>, rot=<0.0012 0.0572 0.0155 -0.9982>}
 calf_l {parent=thigh_l, loc=<-39.0484 0 0>, rot=<0.0006 -0.0161 -0.0391 -0.9991>}
 calf_twist_01_l {parent=calf_l, loc=<-35.813 -0.0967 0.0397>, rot=<0.0237 0.0095 -0.0035 -0.9997>}
 calf_twist_02_l {parent=calf_l, loc=<-18.8064 -0.0967 0.0397>, rot=<0.0237 0.0095 -0.0035 -0.9997>}
 foot_l {parent=calf_l, loc=<-42.561 0.1077 0.4292>, rot=<0.0418 -0.0455 0.0343 -0.9975>}
 ball_l {parent=foot_l, loc=<-6.1599 13.8264 0.6806>, rot=<0.0001 -0.0001 -0.7063 -0.7079>}
 dyn_l_knee {parent=calf_l, loc=<0 0 0>, rot=<0 0 0 1>}
 thigh_twist_01_l {parent=thigh_l, loc=<-21.6303 0 0>, rot=<0.0013 0 0 -1>}
 dyn_shorts_in_l_MSTR {parent=thigh_twist_01_l, loc=<-2.8031 -0.7827 9.4966>, rot=<0.0088 -0.0019 0.998 -0.0618>}
 dyn_shorts_in_l {parent=dyn_shorts_in_l_MSTR, loc=<0 0 0>, rot=<0 0 0 1>}
 dyn_shorts_ft_l_MSTR {parent=thigh_twist_01_l, loc=<-11.3165 8.7128 -2.0869>, rot=<-0.0298 0.7033 0.7094 -0.034>}
 dyn_shorts_ft_l {parent=dyn_shorts_ft_l_MSTR, loc=<0 0 0>, rot=<0 0 0 1>}
 dyn_shorts_out_l_MSTR {parent=thigh_twist_01_l, loc=<-2.7108 -1.6058 -9.4544>, rot=<0.059 -0.9977 0.0007 -0.032>}
 dyn_shorts_out_l {parent=dyn_shorts_out_l_MSTR, loc=<0 0 0>, rot=<0 0 0 -1>}
 dyn_shorts_bk_l_MSTR {parent=thigh_twist_01_l, loc=<4.5328 -10.3507 0.1279>, rot=<0.0198 -0.7103 0.7033 -0.0211>}
 dyn_shorts_bk_l {parent=dyn_shorts_bk_l_MSTR, loc=<0 0 0>, rot=<0 0 0 1>}
 thigh_r {parent=pelvis, loc=<-2.5691 -0.0109 7.7501>, rot=<0.0572 -0.0012 0.9982 0.0155>}
 calf_r {parent=thigh_r, loc=<39.0487 0 0>, rot=<0.0006 -0.0161 -0.0391 -0.9991>}
 calf_twist_01_r {parent=calf_r, loc=<35.8132 0.0967 -0.0397>, rot=<0.0237 0.0095 -0.0035 -0.9997>}
 calf_twist_02_r {parent=calf_r, loc=<18.8064 0.0967 -0.0397>, rot=<0.0237 0.0095 -0.0035 -0.9997>}
 foot_r {parent=calf_r, loc=<42.5612 -0.1077 -0.4292>, rot=<0.0418 -0.0455 0.0343 -0.9975>}
 ball_r {parent=foot_r, loc=<6.1598 -13.8264 -0.6806>, rot=<0.0001 -0.0001 -0.7063 -0.7079>}
 dyn_r_knee {parent=calf_r, loc=<0 0 0>, rot=<0 0 0 1>}
 thigh_twist_01_r {parent=thigh_r, loc=<21.6303 0 0>, rot=<0.0013 0 0 -1>}
 dyn_shorts_ft_r_MSTR {parent=thigh_twist_01_r, loc=<11.3167 -8.7128 2.0869>, rot=<-0.7094 -0.034 -0.0298 -0.7033>}
 dyn_shorts_ft_r {parent=dyn_shorts_ft_r_MSTR, loc=<0 0 0>, rot=<0 0 0 -1>}
 dyn_shorts_out_r_MSTR {parent=thigh_twist_01_r, loc=<2.711 1.6058 9.4544>, rot=<0.0007 0.032 -0.059 -0.9977>}
 dyn_shorts_out_r {parent=dyn_shorts_out_r_MSTR, loc=<0 0 0>, rot=<0 0 0 1>}
 dyn_shorts_bk_r_MSTR {parent=thigh_twist_01_r, loc=<-4.5326 10.3507 -0.1279>, rot=<0.7033 0.0211 -0.0198 -0.7103>}
 dyn_shorts_bk_r {parent=dyn_shorts_bk_r_MSTR, loc=<0 0 0>, rot=<0 0 0 -1>}
 dyn_shorts_in_r_MSTR {parent=thigh_twist_01_r, loc=<2.8109 0.7539 -9.4828>, rot=<0.998 0.0618 -0.0088 -0.0019>}
 dyn_shorts_in_r {parent=dyn_shorts_in_r_MSTR, loc=<0 0 0>, rot=<0 0 0 -1>}
 dyn_shirt_bk_r {parent=pelvis, loc=<-1.4419 -11.4805 10.846>, rot=<-0.0127 -0.707 0.707 0.0127>}
 dyn_shirt_bk_l {parent=pelvis, loc=<-1.4419 -11.4805 -10.8459>, rot=<-0.0127 -0.707 0.707 0.0127>}
 dyn_shirt_bk {parent=pelvis, loc=<-2.3987 -15.2492 -0.9611>, rot=<-0.0127 -0.707 0.707 0.0127>}
]
