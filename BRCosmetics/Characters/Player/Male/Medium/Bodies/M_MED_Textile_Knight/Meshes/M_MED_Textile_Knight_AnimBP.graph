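# Animation Blueprint for the Textile Knight male medium body cosmetic, adding corrective deformation and cloth dynamics on top of the pose copied from the attached parent mesh.
# Control Rigs, knee constraints, additive shorts-bone offsets and RigidBody simulations for pants, collar and body run in sequence.
# Each update derives corrective alphas from leg bone angles, sets state-based gravity and toggles the final target rig by skydive, relaxed and targeting states.
# Dynamics reset on crouch, skydive and parachute events.
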AnimBlueprint /BRCosmetics/Characters/Player/Male/Medium/Bodies/M_MED_Textile_Knight/Meshes/M_MED_Textile_Knight_AnimBP
target_skeleton=&/M_MED_Textile_Knight_Skeleton
graphs [
 AnimGraph {nodes [
   n1 = Get legs_rot_sides() -> (legs_rot_sides:double)
   n2 = "Transform (Modify) Bone - Bone: dyn_shorts_in_r_MSTR"(Alpha=@n1, Translation="0,0,0", Rotation="-10,0.000,0.000", Scale="1,1,1") {
    rotationMode=BMM_Additive
    rotationSpace=BCS_BoneSpace
    translationMode=BMM_Additive
    translationSpace=BCS_BoneSpace
    alphaBoolBlend {blendOption=Linear}
    boneToModify {boneName=dyn_shorts_in_r_MSTR}
    rotation {pitch=-10}
   }
   "Transform (Modify) Bone - Bone: dyn_shorts_in_l_MSTR"(ComponentPose=@n2, Alpha=@n1, Translation="0,0,0", Rotation="-10,0.000,0.000", Scale="1,1,1") {
    rotationMode=BMM_Additive
    rotationSpace=BCS_BoneSpace
    translationMode=BMM_Additive
    translationSpace=BCS_BoneSpace
    alphaBoolBlend {blendOption=Linear}
    boneToModify {boneName=dyn_shorts_in_l_MSTR}
    rotation {pitch=-10}
   }
   n3 = Control Rig(Source=Copy Pose From Mesh() {bUseAttachedParent=true}) {
    alphaBoolBlend {blendOption=Linear}
    controlRigAssetReference {blueprintRigClass=&/M_MED_Textile_Knight_Deform_CtrlRig_twists_3.M_MED_Textile_Knight_Deform_CtrlRig_twists_3_C}
   }
   n4 = Control Rig!(
    Source=@n3
    Alpha
    KneeRot_fr_l_114deg:Vector
    KneeTrans_fr_l_114deg:Vector="0.258,-3.591,0.009"
    KneeScale_fr_l_114deg:Vector="1,1,1"
    KneeScale_fr_l_130deg:Vector="1,1,1"
    KneeRot_fr_l_130deg:Vector="0,0,-51.2"
    KneeTrans_fr_l_130deg:Vector="0.258,-3.591,0.009"
   ) -> (Pose) {
    alphaBoolBlend {blendOption=Linear}
    controlRigAssetReference {blueprintRigClass=&/M_MED_Textile_Knight_Deform_CtrlRig_sculpters_2.M_MED_Textile_Knight_Deform_CtrlRig_sculpters_2_C}
   }
   n5 = "Constraint - dyn_r_knee (RR)"!(ComponentPose=Local To Component(LocalPose=@n4), Alpha=1, ConstraintWeights_0=0.7, ConstraintWeights_1:float=0.3) -> (Pose) {
    alphaBoolBlend {blendOption=Linear}
    boneToModify {boneName=dyn_r_knee}
    constraintSetup [{transformType=Rotation, targetBone {boneName=calf_r}}, {transformType=Rotation, targetBone {boneName=thigh_r}}]
    constraintWeights [1,1]
   }
   n6 = "Constraint - dyn_l_knee (RR)"!(ComponentPose=@n5, Alpha=1, ConstraintWeights_0=0.7, ConstraintWeights_1:float=0.3) -> (Pose) {
    alphaBoolBlend {blendOption=Linear}
    boneToModify {boneName=dyn_l_knee}
    constraintSetup [{transformType=Rotation, targetBone {boneName=calf_l}}, {transformType=Rotation, targetBone {boneName=thigh_l}}]
    constraintWeights [1,1]
   }
   n7 = "Constraint - dyn_shorts_in_l_MSTR (PP)"!(ComponentPose=@n6, Alpha=0, ConstraintWeights_0=0.5, ConstraintWeights_1:float=0.5) -> (Pose) {
    alphaBoolBlend {blendOption=Linear}
    boneToModify {boneName=dyn_shorts_in_l_MSTR}
    constraintSetup [{targetBone {boneName=thigh_twist_01_l}}, {targetBone {boneName=thigh_twist_01_r}}]
    constraintWeights [1,1]
   }
   n8 = "Constraint - dyn_shorts_in_r_MSTR (PP)"!(ComponentPose=@n7, Alpha=0, ConstraintWeights_0=0.5, ConstraintWeights_1:float=0.5) -> (Pose) {
    alphaBoolBlend {blendOption=Linear}
    boneToModify {boneName=dyn_shorts_in_r_MSTR}
    constraintSetup [{targetBone {boneName=thigh_twist_01_l}}, {targetBone {boneName=thigh_twist_01_r}}]
    constraintWeights [1,1]
   }
   n9 = Get calf_rot_r() -> (calf_rot_r:double)
   n10 = "Transform (Modify) Bone - Bone: dyn_shorts_bk_r_MSTR"(ComponentPose=@n8, Alpha=@n9, Translation="-2,-2,-5", Rotation="100,0,0", Scale="1,1,1") {
    rotationMode=BMM_Additive
    rotationSpace=BCS_BoneSpace
    translationMode=BMM_Additive
    translationSpace=BCS_BoneSpace
    alphaBoolBlend {blendOption=Linear}
    boneToModify {boneName=dyn_shorts_bk_r_MSTR}
    rotation {pitch=100}
    translation=<-2 -2 -5>
   }
   n11 = "Transform (Modify) Bone - Bone: dyn_shorts_out_r_MSTR"(ComponentPose=@n10, Alpha=@n9, Translation="0,0,0", Rotation="0,60,0.000", Scale="1,1,1") {
    rotationMode=BMM_Additive
    rotationSpace=BCS_BoneSpace
    translationMode=BMM_Additive
    translationSpace=BCS_BoneSpace
    alphaBoolBlend {blendOption=Linear}
    boneToModify {boneName=dyn_shorts_out_r_MSTR}
    rotation {yaw=60}
   }
   n12 = "Transform (Modify) Bone - Bone: dyn_shorts_in_r_MSTR"(ComponentPose=@n11, Alpha=@n9, Translation="0,0,0", Rotation="0,-40,0", Scale="1,1,1") {
    rotationMode=BMM_Additive
    rotationSpace=BCS_BoneSpace
    translationMode=BMM_Additive
    translationSpace=BCS_BoneSpace
    alphaBoolBlend {blendOption=Linear}
    boneToModify {boneName=dyn_shorts_in_r_MSTR}
    rotation {yaw=-40}
   }
   n13 = Get calf_rot_l() -> (calf_rot_l:double)
   n14 = "Transform (Modify) Bone - Bone: dyn_shorts_bk_l_MSTR"(ComponentPose=@n12, Alpha=@n13, Translation="0,-3,0", Rotation="100,0,0", Scale="1,1,1") {
    rotationMode=BMM_Additive
    rotationSpace=BCS_BoneSpace
    translationMode=BMM_Additive
    translationSpace=BCS_BoneSpace
    alphaBoolBlend {blendOption=Linear}
    boneToModify {boneName=dyn_shorts_bk_l_MSTR}
    rotation {pitch=100}
    translation {y=-3}
   }
   n15 = "Transform (Modify) Bone - Bone: dyn_shorts_out_l_MSTR"(ComponentPose=@n14, Alpha=@n13, Translation="0,0,0", Rotation="0,-60,0.000", Scale="1,1,1") {
    rotationMode=BMM_Additive
    rotationSpace=BCS_BoneSpace
    translationMode=BMM_Additive
    translationSpace=BCS_BoneSpace
    alphaBoolBlend {blendOption=Linear}
    boneToModify {boneName=dyn_shorts_out_l_MSTR}
    rotation {yaw=-60}
   }
   n16 = "Transform (Modify) Bone - Bone: dyn_shorts_in_l_MSTR"(ComponentPose=@n15, Alpha=@n13, Translation="0,0,2", Rotation="0,40,0", Scale="1,1,1") {
    rotationMode=BMM_Additive
    rotationSpace=BCS_BoneSpace
    translationMode=BMM_Additive
    translationSpace=BCS_BoneSpace
    alphaBoolBlend {blendOption=Linear}
    boneToModify {boneName=dyn_shorts_in_l_MSTR}
    rotation {yaw=40}
    translation {z=2}
   }
   n17 = Control Rig!(Source=Component To Local(ComponentPose=@n16), Alpha) -> (Pose) {
    alphaBoolBlend {blendOption=Linear}
    controlRigAssetReference {blueprintRigClass=&/M_MED_Textile_Knight_New_CtrlRig.M_MED_Textile_Knight_New_CtrlRig_C}
   }
   n18 = Get gravity_override() -> (gravity_override:Vector)
   n19 = RigidBody!(ComponentPose=Local To Component(LocalPose=@n17), Alpha, OverrideWorldGravity:Vector=@n18, ExternalForce) -> (Pose) {
    bClampLinearTranslationLimitToRefPose=true
    bOverrideWorldGravity=true
    overridePhysicsAsset=&/M_MED_TExtile_Knight_pants_r
    simulationSpace=BaseBoneSpace
    alphaBoolBlend {blendOption=Linear}
    baseBoneRef {boneName=calf_r}
    componentAppliedLinearAccClamp=<1000 1000 1000>
    componentLinearAccScale=<0.2 0.2 0.2>
    componentLinearVelScale=<0.2 0.2 0.2>
    overrideWorldGravity {z=-980}
   }
   n20 = RigidBody(ComponentPose=@n19, OverrideWorldGravity=@n18) {
    bClampLinearTranslationLimitToRefPose=true
    bOverrideWorldGravity=true
    overridePhysicsAsset=&/M_MED_Textile_Knight_pants_l
    simulationSpace=BaseBoneSpace
    alphaBoolBlend {blendOption=Linear}
    baseBoneRef {boneName=calf_l}
    componentAppliedLinearAccClamp=<1000 1000 1000>
    componentLinearAccScale=<0.2 0.2 0.2>
    componentLinearVelScale=<0.2 0.2 0.2>
    overrideWorldGravity {z=-980}
   }
   n21 = RigidBody!(ComponentPose=@n20, Alpha, ExternalForce) -> (Pose) {
    bClampLinearTranslationLimitToRefPose=true
    bOverrideWorldGravity=true
    overridePhysicsAsset=&/M_MED_Textile_Knight_Physics
    simulationSpace=BaseBoneSpace
    alphaBoolBlend {blendOption=Linear}
    baseBoneRef {boneName=root}
    componentAppliedLinearAccClamp=<3000 3000 3000>
    componentLinearAccScale=<0.3 0.3 0.5>
    componentLinearVelScale=<0.3 0.3 0.5>
    overrideWorldGravity {z=-980}
   }
   n22 = RigidBody(ComponentPose=@n21) {
    bClampLinearTranslationLimitToRefPose=true
    bOverrideWorldGravity=true
    overridePhysicsAsset=&/M_MED_Textile_Knight_Collar_Physics
    simulationSpace=BaseBoneSpace
    alphaBoolBlend {blendOption=Linear}
    baseBoneRef {boneName=spine_05}
    componentAppliedLinearAccClamp=<3000 3000 3000>
    componentLinearAccScale=<0.3 0.3 0.3>
    componentLinearVelScale=<0.3 0.3 0.3>
    overrideWorldGravity {z=-980}
   }
   n23 = Get calf_rot_r()
   n24 = "Transform (Modify) Bone - Bone: dyn_shorts_bk_r"(ComponentPose=@n22, Alpha=@n23, Translation="0,0,0", Rotation="0,0,0", Scale="0.2,1,1") {
    rotationMode=BMM_Additive
    rotationSpace=BCS_BoneSpace
    scaleMode=BMM_Additive
    scaleSpace=BCS_BoneSpace
    translationMode=BMM_Additive
    translationSpace=BCS_BoneSpace
    alphaBoolBlend {blendOption=Linear}
    boneToModify {boneName=dyn_shorts_bk_r}
    scale {x=0.2}
   }
   n25 = "Transform (Modify) Bone - Bone: dyn_shorts_in_r"(ComponentPose=@n24, Alpha=@n23, Translation="0,0,0", Rotation="0,0,0", Scale="0.5,1,1") {
    rotationMode=BMM_Additive
    rotationSpace=BCS_BoneSpace
    scaleMode=BMM_Additive
    scaleSpace=BCS_BoneSpace
    translationMode=BMM_Additive
    translationSpace=BCS_BoneSpace
    alphaBoolBlend {blendOption=Linear}
    boneToModify {boneName=dyn_shorts_in_r}
    scale {x=0.5}
   }
   n26 = "Transform (Modify) Bone - Bone: dyn_shorts_out_r"(ComponentPose=@n25, Alpha=@n23, Translation="0,0,0", Rotation="0,0,0", Scale="0.5,1,1") {
    rotationMode=BMM_Additive
    rotationSpace=BCS_BoneSpace
    scaleMode=BMM_Additive
    scaleSpace=BCS_BoneSpace
    translationMode=BMM_Additive
    translationSpace=BCS_BoneSpace
    alphaBoolBlend {blendOption=Linear}
    boneToModify {boneName=dyn_shorts_out_r}
    scale {x=0.5}
   }
   n27 = "Transform (Modify) Bone - Bone: dyn_shorts_ft_r"(ComponentPose=@n26, Alpha=@n23, Translation="0,0,0", Rotation="-30,0,0", Scale="0.8,1,1") {
    rotationMode=BMM_Additive
    rotationSpace=BCS_BoneSpace
    scaleMode=BMM_Additive
    scaleSpace=BCS_BoneSpace
    translationMode=BMM_Additive
    translationSpace=BCS_BoneSpace
    alphaBoolBlend {blendOption=Linear}
    boneToModify {boneName=dyn_shorts_ft_r}
    rotation {pitch=-30}
    scale {x=0.8}
   }
   n28 = Get calf_rot_l()
   n29 = "Transform (Modify) Bone - Bone: dyn_shorts_bk_l"(ComponentPose=@n27, Alpha=@n28, Translation="0,-3,2", Rotation="0,0,0", Scale="0.2,1,1") {
    rotationMode=BMM_Additive
    rotationSpace=BCS_BoneSpace
    scaleMode=BMM_Additive
    scaleSpace=BCS_BoneSpace
    translationMode=BMM_Additive
    translationSpace=BCS_BoneSpace
    alphaBoolBlend {blendOption=Linear}
    boneToModify {boneName=dyn_shorts_bk_l}
    scale {x=0.2}
    translation {y=-3, z=2}
   }
   n30 = "Transform (Modify) Bone - Bone: dyn_shorts_in_l"(ComponentPose=@n29, Alpha=@n28, Translation="0,0,0", Rotation="0,0,0", Scale="0.45,1,1") {
    rotationMode=BMM_Additive
    rotationSpace=BCS_BoneSpace
    scaleMode=BMM_Additive
    scaleSpace=BCS_BoneSpace
    translationMode=BMM_Additive
    translationSpace=BCS_BoneSpace
    alphaBoolBlend {blendOption=Linear}
    boneToModify {boneName=dyn_shorts_in_l}
    scale {x=0.45}
   }
   n31 = "Transform (Modify) Bone - Bone: dyn_shorts_out_l"(ComponentPose=@n30, Alpha=@n28, Translation="0,0,0", Rotation="0,0,0", Scale="0.45,1,1") {
    rotationMode=BMM_Additive
    rotationSpace=BCS_BoneSpace
    scaleMode=BMM_Additive
    scaleSpace=BCS_BoneSpace
    translationMode=BMM_Additive
    translationSpace=BCS_BoneSpace
    alphaBoolBlend {blendOption=Linear}
    boneToModify {boneName=dyn_shorts_out_l}
    scale {x=0.45}
   }
   n32 = "Transform (Modify) Bone - Bone: dyn_shorts_ft_l"(ComponentPose=@n31, Alpha=@n28, Translation="0,0,0", Rotation="-30,0.000,0.000", Scale="0.8,1,1") {
    rotationMode=BMM_Additive
    rotationSpace=BCS_BoneSpace
    scaleMode=BMM_Additive
    scaleSpace=BCS_BoneSpace
    translationMode=BMM_Additive
    translationSpace=BCS_BoneSpace
    alphaBoolBlend {blendOption=Linear}
    boneToModify {boneName=dyn_shorts_ft_l}
    rotation {pitch=-30}
    scale {x=0.8}
   }
   n33 = Control Rig!(Source=Component To Local(ComponentPose=@n32), bAlphaBoolEnabled:bool=Get ctrl rig() -> (ctrl rig:bool)) -> (Pose) {
    alphaInputType=Bool
    alphaBoolBlend {blendInTime=0.3, blendOption=Linear, blendOutTime=0.3}
    controlRigAssetReference {blueprintRigClass=&/M_MED_Textile_Knight_target_CtrlRig.M_MED_Textile_Knight_target_CtrlRig_C}
   }
   Output Pose(Result=@n33) {layerGroup=None}
  ]}
 EventGraph {nodes [
   Event BlueprintUpdateAnimation()
   > n1 = Sequence!(execute) -> (then_0, then_1, then_2:Exec, then_3:Exec, then_4:Exec, then_5:Exec)
   n2 = BreakTransform(InTransform=GetDeltaTransformFromRefPose(self=GetOwnerSkeletalMeshComponent(), BoneName=thigh_l, BaseName=pelvis))
   > Set thigh_up_l(thigh_up_l:double=MapRangeClamped(Value=BreakRotator(InRot=@n2.Rotation).Yaw, InRangeA=-20, InRangeB=-100)) -> (Output_Get:double)
   n3 = BreakTransform(InTransform=GetDeltaTransformFromRefPose(self=GetOwnerSkeletalMeshComponent(), BoneName=thigh_r, BaseName=pelvis))
   > Set thigh_up_r(thigh_up_r:double=MapRangeClamped(Value=BreakRotator(InRot=@n3.Rotation).Yaw, InRangeA=-20, InRangeB=-100)) -> (Output_Get:double)
   n4 = BreakTransform(InTransform=GetDeltaTransformFromRefPose(self=GetOwnerSkeletalMeshComponent(), BoneName=thigh_r, BaseName=thigh_l))
   > Set legs_rot_sides(legs_rot_sides:double=MapRangeClamped(Value=BreakRotator(InRot=@n4.Rotation).Pitch, InRangeA=0, InRangeB=16)) -> (Output_Get:double)
   n5 = BreakTransform(InTransform=GetDeltaTransformFromRefPose(self=GetOwnerSkeletalMeshComponent(), BoneName=thigh_r, BaseName=calf_r))
   > Set calf_rot_r(calf_rot_r:double=MapRangeClamped(Value=BreakRotator(InRot=@n5.Rotation).Yaw, InRangeA=-20, InRangeB=-120)) -> (Output_Get:double)
   n6 = BreakTransform(InTransform=GetDeltaTransformFromRefPose(self=GetOwnerSkeletalMeshComponent(), BoneName=thigh_l, BaseName=calf_l))
   > Set calf_rot_l(calf_rot_l:double=MapRangeClamped(Value=BreakRotator(InRot=@n6.Rotation).Yaw, InRangeA=-10, InRangeB=-120)) -> (Output_Get:double)
   n7 = Get bIsCrouching()
   Set isNotCrouching(execute=@n1.then_1, isNotCrouching:double="To Float (Boolean)"(InBool=NOT Boolean(A=@n7))) -> (Output_Get:double)
   Branch(execute=@n1.then_2, Condition=@n7)
   > Set legs_rot_sides(legs_rot_sides=0.5)
   n8 = Gravity Override State Function DEPRECATED(
    execute=@n1.then_3
    CharacterPartAnimInstance=Self-Reference()
    GameTimeInSeconds=GetGameTimeInSeconds()
    GravityOverrideParams=Get Gravity Override Params()
   )
   > Set gravity_override(gravity_override=@n8)
   n9 = Branch(
    execute=@n1.then_4
    Condition=OR Boolean!(A=Get bIsSkyDiving(), B=Get bIsRelaxedLevel1() -> (bIsRelaxedLevel1:bool), C:bool=Get bIsRelaxedLevel2() -> (bIsRelaxedLevel2:bool)) -> (ReturnValue)
   )
   > Set ctrl rig(ctrl rig:bool=false) -> (Output_Get:bool)
   Set ctrl rig(execute=@n9.else, ctrl rig=true)
   Branch(execute=@n1.then_5, Condition=Get bIsTargeting())
   > Set ctrl rig(ctrl rig=true)
   n10 = Event OnBeginCrouching()
   ResetDynamics(execute=[@n11.then, @n10.then, @n12.then, @n13.then], InTeleportType=TeleportPhysics)
   > ResetDynamics(InTeleportType=ResetPhysics)
   Event OnParachuteOpened()
   > n11 = Delay(Duration=1.5)
   n12 = Event OnEndSkydiving()
   Event BlueprintInitializeAnimation()
   > Set Body_Mesh(Body_Mesh=GetOwnerSkeletalMeshComponent())
   > ResetDynamics(InTeleportType=TeleportPhysics)
   n13 = Event OnBeginSkydiving()
  ]}
]
variables [
 Body_Mesh {type=SkeletalMeshComponent}
 calf_rot_l {type=number}
 calf_rot_r {type=number}
 ctrl rig {type=boolean}
 Gravity Override Params {
  type=GravityOverrideParamsStruct
  default {
   emoteOrMelee {gravityOverride=<0 0 -600>}
   front_End {gravityOverride=<0 0 -600>}
   onGround_Moving {gravityOverride=<0 0 -600>}
   onGround_Standing {gravityOverride=<0 0 -600>}
   skydive_GoingUp {gravityOverride=<0 0 -200>}
   skydive_Parachuting {gravityOverride=<0 0 -600>}
   swimming {gravityOverride=<0 0 0>}
  }
 }
 gravity_override {type=Vector}
 isNotCrouching {type=number}
 legs_rot_sides {type=number}
 thigh_up_l {type=number}
 thigh_up_r {type=number}
]
